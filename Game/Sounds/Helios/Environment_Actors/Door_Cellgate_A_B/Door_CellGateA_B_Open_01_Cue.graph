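# Sound cue for the opening sound of a cell gate door among the Helios environment actors.
# A Random node picks one of three equally weighted open waves and feeds a Modulator that applies slight pitch variation.
# All four QualityLevel inputs share that same chain, so every quality tier sounds identical.
# It uses the Fort_Doors sound class and shared door attenuation, with a concurrency override that prevents new instances at the limit.
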SoundCue /Game/Sounds/Helios/Environment_Actors/Door_Cellgate_A_B/Door_CellGateA_B_Open_01_Cue
root=QualityLevel_0
nodes [
 {id=Modulator_0, children [Random_0], properties {pitchMax=0.99, pitchMin=1.01}}
 {id=QualityLevel_0, children [Modulator_0,Modulator_0,Modulator_0,Modulator_0]}
 {id=Random_0, children [WavePlayer_0,WavePlayer_2,WavePlayer_3], properties {weights [1,1,1]}}
 {id=WavePlayer_0, properties {soundWaveAssetPtr=&/Door_CellGateA_B_Open_01}}
 {id=WavePlayer_2, properties {soundWaveAssetPtr=&/Door_CellGateA_Open_02}}
 {id=WavePlayer_3, properties {soundWaveAssetPtr=&/Door_CellGateA_Open_03}}
]
properties {
 attenuationSettings=/Game/Sounds/Attenuation/Fort_Doors_Att
 bOverrideConcurrency=true
 maxDistance=2400
 soundClassObject=&Fort_Doors
 volumeMultiplier=0.6
 attenuationOverrides {falloffDistance=1600}
 concurrencyOverrides {resolutionRule=PreventNew}
}
